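# Sound cue for the music played when the Fire Petal Shrine is activated, in its air sprite variant.
# A Random node picks one of two equally weighted music waves behind a fixed delay of a few seconds.
# A root Enveloper wraps the delayed branch, holding full volume at first and then tapering to about half.
# It uses the shrine's activation-music attenuation asset and the loot-reward sound class.
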SoundCue /FirePetalShrineGameplay/Assets/Sounds/Cues/SC_FirePetalShrine_Activated_Air_Music
root=Enveloper_0
nodes [
 {id=Delay_1, children [Random_0], properties {delayMax=2.5, delayMin=2.5}}
 {
  id=Enveloper_0
  children [Delay_1]
  properties {
   pitchCurve {editorCurveData {keys [{time=0}]}}
   volumeCurve {editorCurveData {keys [<at 0 1>,<at 5 1>,<at 8 0.7>,<at 10 0.6>,<at 12.7008 0.469>]}}
  }
 }
 {id=Random_0, children [WavePlayer_3,WavePlayer_4], properties {bShouldExcludeFromBranchCulling=true, weights [1,1]}}
 {id=WavePlayer_3, properties {soundWaveAssetPtr=/FirePetalShrineGameplay/Assets/Sounds/Waves/FirePetalShrine_AirSprite_Activate_Music_01}}
 {id=WavePlayer_4, properties {soundWaveAssetPtr=/FirePetalShrineGameplay/Assets/Sounds/Waves/FirePetalShrine_AirSprite_Activate_Music_02}}
]
properties {
 attenuationSettings=/FirePetalShrineGameplay/Assets/Sounds/Settings/AS_FirePetalShrine_Activate_Music
 maxDistance=6100
 soundClassObject=&Fort_Loot_Reward_PSM
 volumeMultiplier=0.7
 attenuationOverrides {bAttenuate=false, bEnableReverbSend=false, bSpatialize=false}
 concurrencyOverrides {resolutionRule=PreventNew}
}
