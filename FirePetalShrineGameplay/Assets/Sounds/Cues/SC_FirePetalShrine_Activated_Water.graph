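# Sound cue for the water variant of the Fire Petal Shrine activation sound.
# A random node picks one of three equally weighted water waves and feeds a modulator at the root for playback variation.
# It uses the shrine's activate attenuation asset and the loot reward sound class.
# Overridden concurrency allows one instance per owner and blocks new plays while it is active.
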SoundCue /FirePetalShrineGameplay/Assets/Sounds/Cues/SC_FirePetalShrine_Activated_Water
root=Modulator_0
nodes [
 {id=Modulator_0, children [Random_0]}
 {id=Random_0, children [WavePlayer_0,WavePlayer_1,WavePlayer_2], properties {bShouldExcludeFromBranchCulling=true, weights [1,1,1]}}
 {id=WavePlayer_0, properties {soundWaveAssetPtr=/FirePetalShrineGameplay/Assets/Sounds/Waves/FirePetalShrine_Activated_Water_01}}
 {id=WavePlayer_1, properties {soundWaveAssetPtr=/FirePetalShrineGameplay/Assets/Sounds/Waves/FirePetalShrine_Activated_Water_02}}
 {id=WavePlayer_2, properties {soundWaveAssetPtr=/FirePetalShrineGameplay/Assets/Sounds/Waves/FirePetalShrine_Activated_Water_03}}
]
properties {
 attenuationSettings=/FirePetalShrineGameplay/Assets/Sounds/Settings/AS_FirePetalShrine_Activate
 bOverrideConcurrency=true
 maxDistance=8600
 soundClassObject=&Fort_Loot_Reward_PSM
 volumeMultiplier=0.8
 attenuationOverrides {bAttenuate=false, bEnableReverbSend=false, bSpatialize=false}
 concurrencyOverrides {bLimitToOwner=true, maxCount=1, resolutionRule=PreventNew}
}
