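# Looping idle ambience Sound Cue for the activated Fire Petal Shrine's water sound.
# A Mixer layers a randomly chosen one of two equally weighted Water Sprite shrine activation loops with a slightly quieter mid ambient shrine loop.
# Each wave player loops on its own, and the random branch is excluded from branch culling.
# The shrine's activated-ambient attenuation asset and a loot-reward sound class govern playback.
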SoundCue /FirePetalShrineGameplay/Assets/Sounds/Cues/SC_FirePetalShrine_Activated_Water_Idle_Loop
root=Mixer_0
nodes [
 {id=Mixer_0, children [Random_1,WavePlayer_1], properties {inputVolume [1,0.736]}}
 {id=Random_1, children [WavePlayer_0,WavePlayer_3], properties {bShouldExcludeFromBranchCulling=true, weights [1,1]}}
 {
  id=WavePlayer_0
  properties {bLooping=true, soundWaveAssetPtr=/FirePetalShrineGameplay/Assets/Sounds/Waves/Water_Sprite_Shrine_Activation_SFX_Loop_01}
 }
 {id=WavePlayer_1, properties {bLooping=true, soundWaveAssetPtr=/FirePetalShrineGameplay/Assets/Sounds/Waves/FirePetalShrine_Ambient_Mid_Loop}}
 {
  id=WavePlayer_3
  properties {bLooping=true, soundWaveAssetPtr=/FirePetalShrineGameplay/Assets/Sounds/Waves/Water_Sprite_Shrine_Activation_SFX_Loop_02}
 }
]
properties {
 attenuationSettings=/FirePetalShrineGameplay/Assets/Sounds/Settings/AS_FirePetalShrine_Activated_Ambient
 maxDistance=5500
 soundClassObject=&Fort_Loot_Reward_PSM
 volumeMultiplier=0.5
 attenuationOverrides {bAttenuate=false, bEnableReverbSend=false, bSpatialize=false}
 concurrencyOverrides {resolutionRule=PreventNew}
}
