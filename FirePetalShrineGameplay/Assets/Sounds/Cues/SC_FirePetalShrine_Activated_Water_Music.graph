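# Sound cue for the Fire Petal Shrine water activation music, playing one of two WaterSprite music stings.
# A Random node picks a wave with equal weight, feeds a fixed 2.5-second Delay, then an Enveloper that holds full volume before gradually fading to about half.
# The Random node is excluded from branch culling, and the cue uses a loot-reward sound class with a shrine music attenuation asset.
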SoundCue /FirePetalShrineGameplay/Assets/Sounds/Cues/SC_FirePetalShrine_Activated_Water_Music
root=Enveloper_0
nodes [
 {id=Delay_1, children [Random_0], properties {delayMax=2.5, delayMin=2.5}}
 {
  id=Enveloper_0
  children [Delay_1]
  properties {
   pitchCurve {editorCurveData {keys [{time=0}]}}
   volumeCurve {editorCurveData {keys [<at 0 1>,<at 5 1>,<at 8.6314 0.8423>,<at 10.8853 0.6557>,<at 13.3159 0.4951>]}}
  }
 }
 {id=Random_0, children [WavePlayer_3,WavePlayer_0], properties {bShouldExcludeFromBranchCulling=true, weights [1,1]}}
 {id=WavePlayer_0, properties {soundWaveAssetPtr=/FirePetalShrineGameplay/Assets/Sounds/Waves/FirePetalShrine_WaterSprite_Activate_Music_02}}
 {id=WavePlayer_3, properties {soundWaveAssetPtr=/FirePetalShrineGameplay/Assets/Sounds/Waves/FirePetalShrine_WaterSprite_Activate_Music_01}}
]
properties {
 attenuationSettings=/FirePetalShrineGameplay/Assets/Sounds/Settings/AS_FirePetalShrine_Activate_Music
 maxDistance=6100
 soundClassObject=&Fort_Loot_Reward_PSM
 volumeMultiplier=1
 attenuationOverrides {bAttenuate=false, bEnableReverbSend=false, bSpatialize=false}
 concurrencyOverrides {resolutionRule=PreventNew}
}
