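# Sound Cue for the looping water sound effect of an activated Fire Petal Shrine.
# A Random node at the root picks between two looping Water Sprite shrine activation waves with equal weight.
# It references the shrine's activation attenuation asset and routes to the Fort_Loot_Reward_PSM sound class.
# The Random node is excluded from branch culling.
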SoundCue /FirePetalShrineGameplay/Assets/Sounds/Cues/SC_FirePetalShrine_Activated_Water_SFX_Loop
root=Random_1
nodes [
 {id=Random_1, children [WavePlayer_0,WavePlayer_3], properties {bShouldExcludeFromBranchCulling=true, weights [1,1]}}
 {
  id=WavePlayer_0
  properties {bLooping=true, soundWaveAssetPtr=/FirePetalShrineGameplay/Assets/Sounds/Waves/Water_Sprite_Shrine_Activation_SFX_Loop_01}
 }
 {
  id=WavePlayer_3
  properties {bLooping=true, soundWaveAssetPtr=/FirePetalShrineGameplay/Assets/Sounds/Waves/Water_Sprite_Shrine_Activation_SFX_Loop_02}
 }
]
properties {
 attenuationSettings=/FirePetalShrineGameplay/Assets/Sounds/Settings/AS_FirePetalShrine_Activate
 maxDistance=8600
 soundClassObject=&Fort_Loot_Reward_PSM
 volumeMultiplier=1
 attenuationOverrides {bAttenuate=false, bEnableReverbSend=false, bSpatialize=false}
 concurrencyOverrides {resolutionRule=PreventNew}
}
