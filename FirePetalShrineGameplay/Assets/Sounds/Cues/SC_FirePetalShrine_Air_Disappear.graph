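# Sound cue for the disappearance sound of the air sprite variant of the Fire Petal Shrine.
# A Random node picks one of four equally weighted wave variations and feeds it through a Modulator at default settings.
# Spatial behavior comes from the shared shrine Activate attenuation asset, and the cue plays in a loot-reward sound class.
# Overridden concurrency allows one instance per owner and rejects new plays while it is active.
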SoundCue /FirePetalShrineGameplay/Assets/Sounds/Cues/SC_FirePetalShrine_Air_Disappear
root=Modulator_0
nodes [
 {id=Modulator_0, children [Random_0]}
 {
  id=Random_0
  children [WavePlayer_0,WavePlayer_1,WavePlayer_2,WavePlayer_3]
  properties {bShouldExcludeFromBranchCulling=true, weights [1,1,1,1]}
 }
 {id=WavePlayer_0, properties {soundWaveAssetPtr=/FirePetalShrineGameplay/Assets/Sounds/Waves/AirSprite_Shrine_Disappear_01}}
 {id=WavePlayer_1, properties {soundWaveAssetPtr=/FirePetalShrineGameplay/Assets/Sounds/Waves/AirSprite_Shrine_Disappear_02}}
 {id=WavePlayer_2, properties {soundWaveAssetPtr=/FirePetalShrineGameplay/Assets/Sounds/Waves/AirSprite_Shrine_Disappear_03}}
 {id=WavePlayer_3, properties {soundWaveAssetPtr=/FirePetalShrineGameplay/Assets/Sounds/Waves/AirSprite_Shrine_Disappear_04}}
]
properties {
 attenuationSettings=/FirePetalShrineGameplay/Assets/Sounds/Settings/AS_FirePetalShrine_Activate
 bOverrideConcurrency=true
 maxDistance=8600
 soundClassObject=&Fort_Loot_Reward_PSM
 volumeMultiplier=1.2
 attenuationOverrides {bAttenuate=false, bEnableReverbSend=false, bSpatialize=false}
 concurrencyOverrides {bLimitToOwner=true, maxCount=1, resolutionRule=PreventNew}
}
