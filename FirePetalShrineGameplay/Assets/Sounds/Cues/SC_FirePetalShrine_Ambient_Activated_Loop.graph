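# Looping ambient sound cue for the Fire Petal Shrine in its activated state.
# A single looping wave player feeds the root with the boon shrine activated ambience wave.
# A dedicated attenuation asset shapes it in the world, and it routes through the Athena ambient sound class.
# A concurrency override caps simultaneous instances, stopping the farthest before preventing new ones.
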SoundCue /FirePetalShrineGameplay/Assets/Sounds/Cues/SC_FirePetalShrine_Ambient_Activated_Loop
root=WavePlayer_3
nodes [{
  id=WavePlayer_3
  properties {bLooping=true, soundWaveAssetPtr=/FirePetalShrineGameplay/Assets/Sounds/Waves/Boon_Shrine_Ambient_Activated_Loop}
 }]
properties {
 attenuationSettings=/FirePetalShrineGameplay/Assets/Sounds/Settings/AS_FirePetalShrine_Activated_Ambient
 bOverrideConcurrency=true
 maxDistance=5500
 soundClassObject=&Fort_Ambient_Athena
 volumeMultiplier=1
 attenuationOverrides {bAttenuate=false, bEnableReverbSend=false, bSpatialize=false}
 concurrencyOverrides {maxCount=3, resolutionRule=StopFarthestThenPreventNew, volumeScaleAttackTime=0.514}
}
