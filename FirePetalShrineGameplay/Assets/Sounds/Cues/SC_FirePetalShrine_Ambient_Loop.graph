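# Sound cue for the looping ambient sound of the Fire Petal Shrine, in the Athena ambient sound class.
# Three looping waves act as close, mid and far layers, each through its own attenuation node, into a mixer weighted toward the close layer.
# The far layer uses a Boon Shrine wave with the general shrine attenuation.
# A concurrency override caps simultaneous instances, stopping the farthest first.
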SoundCue /FirePetalShrineGameplay/Assets/Sounds/Cues/SC_FirePetalShrine_Ambient_Loop
root=Mixer_1
nodes [
 {
  id=Attenuation_0
  children [WavePlayer_1]
  properties {attenuationSettings=/FirePetalShrineGameplay/Assets/Sounds/Settings/AS_FirePetalShrine_Ambient_Close}
 }
 {
  id=Attenuation_1
  children [WavePlayer_0]
  properties {attenuationSettings=/FirePetalShrineGameplay/Assets/Sounds/Settings/AS_FirePetalShrine_Ambient_Mid}
 }
 {
  id=Attenuation_2
  children [WavePlayer_2]
  properties {attenuationSettings=/FirePetalShrineGameplay/Assets/Sounds/Settings/AS_FirePetalShrine_Ambient}
 }
 {id=Mixer_1, children [Attenuation_0, Attenuation_1, Attenuation_2], properties {inputVolume [2.6452,0.3946,1.1793]}}
 {id=WavePlayer_0, properties {bLooping=true, soundWaveAssetPtr=/FirePetalShrineGameplay/Assets/Sounds/Waves/FirePetalShrine_Ambient_Mid_Loop}}
 {id=WavePlayer_1, properties {bLooping=true, soundWaveAssetPtr=/FirePetalShrineGameplay/Assets/Sounds/Waves/FirePetalShrine_Ambient_Loop}}
 {id=WavePlayer_2, properties {bLooping=true, soundWaveAssetPtr=/FirePetalShrineGameplay/Assets/Sounds/Waves/Boon_Shrine_Ambient_Far_Loop}}
]
properties {
 bOverrideConcurrency=true
 maxDistance=7500
 soundClassObject=&Fort_Ambient_Athena
 volumeMultiplier=0.8
 attenuationOverrides {bAttenuate=false, bEnableReverbSend=false, bSpatialize=false}
 concurrencyOverrides {maxCount=3, resolutionRule=StopFarthestThenPreventNew, volumeScaleAttackTime=0.514}
}
