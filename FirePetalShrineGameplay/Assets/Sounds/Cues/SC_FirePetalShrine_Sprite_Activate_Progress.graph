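# Sound cue for the Fire Petal Shrine sprite's activation progress, built on a single activation-progress wave.
# The wave player feeds a default Modulator, then a root Enveloper whose volume curve rises from half to full over ten seconds.
# Playback uses the loot reward sound class, the shrine's activation attenuation asset and its AmbientDuck concurrency set.
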SoundCue /FirePetalShrineGameplay/Assets/Sounds/Cues/SC_FirePetalShrine_Sprite_Activate_Progress
root=Enveloper_0
nodes [
 {
  id=Enveloper_0
  children [Modulator_0]
  properties {pitchCurve {editorCurveData {keys [{time=0}]}}, volumeCurve {editorCurveData {keys [<at 0 0.5>,<at 5 0.75>,<at 10 1>]}}}
 }
 {id=Modulator_0, children [WavePlayer_1]}
 {id=WavePlayer_1, properties {soundWaveAssetPtr=/FirePetalShrineGameplay/Assets/Sounds/Waves/FirePetalShrine_Activate_Progress}}
]
properties {
 attenuationSettings=/FirePetalShrineGameplay/Assets/Sounds/Settings/AS_FirePetalShrine_Activate
 maxDistance=8600
 priority=2
 soundClassObject=&Fort_Loot_Reward_PSM
 volumeMultiplier=1.4
 attenuationOverrides {bAttenuate=false, bEnableReverbSend=false, bSpatialize=false}
 concurrencyOverrides {resolutionRule=PreventNew}
 concurrencySet [/FirePetalShrineGameplay/Assets/Sounds/Settings/CS_FirePetalShrine_AmbientDuck]
}
