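# Sound cue for the start of the Fire Petal Shrine sprite's activation sound.
# A random node picks one of four equally weighted start waves, and a modulator pitches it slightly down by a varying amount.
# The cue boosts volume, references the shrine's activation attenuation settings and routes through the loot reward sound class.
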SoundCue /FirePetalShrineGameplay/Assets/Sounds/Cues/SC_FirePetalShrine_Sprite_Activate_Start
root=Modulator_0
nodes [
 {id=Modulator_0, children [Random_0], properties {pitchMax=0.95, pitchMin=0.85}}
 {id=Random_0, children [WavePlayer_1,WavePlayer_2,WavePlayer_3,WavePlayer_4], properties {weights [1,1,1,1]}}
 {id=WavePlayer_1, properties {soundWaveAssetPtr=/FirePetalShrineGameplay/Assets/Sounds/Waves/FirePetalShrine_Activate_Start_01}}
 {id=WavePlayer_2, properties {soundWaveAssetPtr=/FirePetalShrineGameplay/Assets/Sounds/Waves/FirePetalShrine_Activate_Start_02}}
 {id=WavePlayer_3, properties {soundWaveAssetPtr=/FirePetalShrineGameplay/Assets/Sounds/Waves/FirePetalShrine_Activate_Start_03}}
 {id=WavePlayer_4, properties {soundWaveAssetPtr=/FirePetalShrineGameplay/Assets/Sounds/Waves/FirePetalShrine_Activate_Start_04}}
]
properties {
 attenuationSettings=/FirePetalShrineGameplay/Assets/Sounds/Settings/AS_FirePetalShrine_Activate
 maxDistance=8600
 soundClassObject=&Fort_Loot_Reward_PSM
 volumeMultiplier=1.3
 attenuationOverrides {bAttenuate=false, bEnableReverbSend=false, bSpatialize=false}
 concurrencyOverrides {resolutionRule=PreventNew}
}
