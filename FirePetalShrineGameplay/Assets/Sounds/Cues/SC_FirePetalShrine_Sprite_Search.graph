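# Sound cue for the search sound of the Fire Petal Shrine sprite, assigned to the loot reward sound class.
# A Random node picks one of four equally weighted search wave variations and passes it through a Modulator at the root.
# Playback uses the shrine's search attenuation asset and a boosted cue volume.
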SoundCue /FirePetalShrineGameplay/Assets/Sounds/Cues/SC_FirePetalShrine_Sprite_Search
root=Modulator_0
nodes [
 {id=Modulator_0, children [Random_1]}
 {
  id=Random_1
  children [WavePlayer_0,WavePlayer_5,WavePlayer_6,WavePlayer_7]
  properties {bShouldExcludeFromBranchCulling=true, weights [1,1,1,1]}
 }
 {id=WavePlayer_0, properties {soundWaveAssetPtr=/FirePetalShrineGameplay/Assets/Sounds/Waves/FirePetalShrine_Sprite_Search_01}}
 {id=WavePlayer_5, properties {soundWaveAssetPtr=/FirePetalShrineGameplay/Assets/Sounds/Waves/FirePetalShrine_Sprite_Search_02}}
 {id=WavePlayer_6, properties {soundWaveAssetPtr=/FirePetalShrineGameplay/Assets/Sounds/Waves/FirePetalShrine_Sprite_Search_03}}
 {id=WavePlayer_7, properties {soundWaveAssetPtr=/FirePetalShrineGameplay/Assets/Sounds/Waves/FirePetalShrine_Sprite_Search_04}}
]
properties {
 attenuationSettings=/FirePetalShrineGameplay/Assets/Sounds/Settings/AS_FirePetalShrine_Search
 maxDistance=5300
 soundClassObject=&Fort_Loot_Reward_PSM
 volumeMultiplier=1.5
 attenuationOverrides {bAttenuate=false, bEnableReverbSend=false, bSpatialize=false}
 concurrencyOverrides {resolutionRule=PreventNew}
}
