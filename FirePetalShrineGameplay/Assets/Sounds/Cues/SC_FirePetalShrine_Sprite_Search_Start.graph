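# Sound cue for the Fire Petal Shrine sprite, named for the start of its search.
# A single wave player plays the Search_Start wave through a Modulator at default settings, which adds slight random variation.
# It uses the shrine's shared ambient attenuation asset and the loot reward sound class, with slightly reduced volume.
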SoundCue /FirePetalShrineGameplay/Assets/Sounds/Cues/SC_FirePetalShrine_Sprite_Search_Start
root=Modulator_0
nodes [
 {id=Modulator_0, children [WavePlayer_0]}
 {id=WavePlayer_0, properties {soundWaveAssetPtr=/FirePetalShrineGameplay/Assets/Sounds/Waves/FirePetalShrine_Search_Start}}
]
properties {
 attenuationSettings=/FirePetalShrineGameplay/Assets/Sounds/Settings/AS_FirePetalShrine_Ambient
 maxDistance=7500
 soundClassObject=&Fort_Loot_Reward_PSM
 volumeMultiplier=0.9
 attenuationOverrides {bAttenuate=false, bEnableReverbSend=false, bSpatialize=false}
 concurrencyOverrides {resolutionRule=PreventNew}
}
