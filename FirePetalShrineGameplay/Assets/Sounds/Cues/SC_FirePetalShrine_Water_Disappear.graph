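# Sound cue for the water sprite shrine's disappear effect in the Fire Petal Shrine gameplay content.
# A Random node picks one of four equally weighted disappear waves and feeds a Modulator at default settings for slight variation.
# It uses the shrine's Activate attenuation asset and a loot reward sound class.
# Overridden concurrency allows one instance per owner and blocks new plays while one is active.
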SoundCue /FirePetalShrineGameplay/Assets/Sounds/Cues/SC_FirePetalShrine_Water_Disappear
root=Modulator_0
nodes [
 {id=Modulator_0, children [Random_1]}
 {
  id=Random_1
  children [WavePlayer_8, WavePlayer_9, WavePlayer_10, WavePlayer_11]
  properties {bShouldExcludeFromBranchCulling=true, weights [1,1,1,1]}
 }
 {id=WavePlayer_10, properties {soundWaveAssetPtr=/FirePetalShrineGameplay/Assets/Sounds/Waves/WaterSprite_Shrine_Disappear_02}}
 {id=WavePlayer_11, properties {soundWaveAssetPtr=/FirePetalShrineGameplay/Assets/Sounds/Waves/WaterSprite_Shrine_Disappear_01}}
 {id=WavePlayer_8, properties {soundWaveAssetPtr=/FirePetalShrineGameplay/Assets/Sounds/Waves/WaterSprite_Shrine_Disappear_04}}
 {id=WavePlayer_9, properties {soundWaveAssetPtr=/FirePetalShrineGameplay/Assets/Sounds/Waves/WaterSprite_Shrine_Disappear_03}}
]
properties {
 attenuationSettings=/FirePetalShrineGameplay/Assets/Sounds/Settings/AS_FirePetalShrine_Activate
 bOverrideConcurrency=true
 maxDistance=8600
 soundClassObject=&Fort_Loot_Reward_PSM
 volumeMultiplier=1.2
 attenuationOverrides {bAttenuate=false, bEnableReverbSend=false, bSpatialize=false}
 concurrencyOverrides {bLimitToOwner=true, maxCount=1, resolutionRule=PreventNew}
}
