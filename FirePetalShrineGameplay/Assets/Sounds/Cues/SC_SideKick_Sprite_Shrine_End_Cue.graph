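# Sound cue for the SideKick Sprite shrine end sound in the FirePetalShrineGameplay plugin, routed through the Player Interacts sound class.
# A single wave player feeds the SideKick_Sprite_Shrine_End wave to the output with a raised volume multiplier.
# Overridden attenuation uses long-range logarithmic falloff and a distance-driven low-pass filter with log frequency scaling.
# It also sets an occlusion filter frequency and a custom occlusion trace channel.
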SoundCue /FirePetalShrineGameplay/Assets/Sounds/Cues/SC_SideKick_Sprite_Shrine_End_Cue
root=WavePlayer_0
nodes [{id=WavePlayer_0, properties {soundWaveAssetPtr=/FirePetalShrineGameplay/Assets/Sounds/Waves/SideKick_Sprite_Shrine_End}}]
properties {
 bOverrideAttenuation=true
 maxDistance=16000
 soundClassObject=/Game/Sounds/FortSoundClassesAndModes/Fort_08_Player_Interacts
 volumeMultiplier=1.3
 attenuationOverrides {
  bAttenuateWithLPF=true
  bEnableLogFrequencyScaling=true
  distanceAlgorithm=Logarithmic
  falloffDistance=14000
  lPFFrequencyAtMax=100
  lPFRadiusMax=2500
  lPFRadiusMin=400
  occlusionLowPassFilterFrequency=5000
  occlusionTraceChannel=ECC_GameTraceChannel11
  attenuationShapeExtents {x=2000}
 }
}
