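# Sound cue for the looping ambience of the SideKick sprite shrine in Fire Petal Shrine gameplay, routed to the player-interacts sound class.
# A single wave player feeds the loop wave straight to the output at a boosted volume.
# Inline attenuation overrides apply logarithmic falloff, distance-based low-pass with a custom air-absorption curve, and a non-spatialized 2D blend up close.
# A concurrency override limits simultaneous instances to three.
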SoundCue /FirePetalShrineGameplay/Assets/Sounds/Cues/SC_SideKick_Sprite_Shrine_Loop_Cue
root=WavePlayer_0
nodes [{id=WavePlayer_0, properties {soundWaveAssetPtr=/FirePetalShrineGameplay/Assets/Sounds/Waves/SideKick_Sprite_Shrine_Loop}}]
properties {
 bOverrideAttenuation=true
 bOverrideConcurrency=true
 maxDistance=16000
 soundClassObject=/Game/Sounds/FortSoundClassesAndModes/Fort_08_Player_Interacts
 volumeMultiplier=1.4
 attenuationOverrides {
  bAttenuateWithLPF=true
  bEnableLogFrequencyScaling=true
  binauralRadius=800
  distanceAlgorithm=Logarithmic
  falloffDistance=12000
  lPFFrequencyAtMax=100
  lPFRadiusMax=2500
  lPFRadiusMin=400
  nonSpatializedRadiusMode=Direct2D
  nonSpatializedRadiusStart=300
  stereoSpread=512
  attenuationShapeExtents {x=4000}
  customLowpassAirAbsorptionCurve {editorCurveData {keys [
     <at 0 1 RCIM_Cubic>
     {arriveTangent=-6.6702, interpMode=RCIM_Cubic, leaveTangent=-6.6702, tangentMode=RCTM_User, time=0.1251, value=0.6991}
     {arriveTangent=-0.952, interpMode=RCIM_Cubic, leaveTangent=-0.952, tangentMode=RCTM_User, time=0.2905, value=0.3065}
     {arriveTangent=-0.6103, interpMode=RCIM_Cubic, leaveTangent=-0.7033, tangentMode=RCTM_User, time=1, value=0.0085}
    ]}}
 }
 concurrencyOverrides {maxCount=3}
}
