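# Sound cue for the sidekick sprite's shrine start sound in the Fire Petal Shrine gameplay feature.
# A single Wave Player feeds the output directly, with boosted volume and no randomization or mixing.
# Cue-level attenuation overrides give it long-range logarithmic falloff beyond a large inner radius.
# Distance-based low-pass filtering follows a custom, steeply dropping absorption curve to muffle the sound as the listener moves away.
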SoundCue /FirePetalShrineGameplay/Assets/Sounds/Cues/SC_SideKick_Sprite_Shrine_Start_Cue
root=WavePlayer_0
nodes [{id=WavePlayer_0, properties {soundWaveAssetPtr=/FirePetalShrineGameplay/Assets/Sounds/Waves/SideKick_Sprite_Shrine_Start}}]
properties {
 bOverrideAttenuation=true
 maxDistance=16000
 volumeMultiplier=2
 attenuationOverrides {
  absorptionMethod=CustomCurve
  bAttenuateWithLPF=true
  distanceAlgorithm=Logarithmic
  falloffDistance=12000
  lPFFrequencyAtMax=100
  lPFRadiusMax=5000
  lPFRadiusMin=2000
  occlusionTraceChannel=ECC_GameTraceChannel11
  attenuationShapeExtents {x=4000}
  customLowpassAirAbsorptionCurve {editorCurveData {keys [
     <at 0 1 RCIM_Cubic>
     {arriveTangent=-6.0417, interpMode=RCIM_Cubic, leaveTangent=-6.0418, tangentMode=RCTM_User, time=0.049, value=0.4692}
     {arriveTangent=-0.7575, interpMode=RCIM_Cubic, leaveTangent=-0.7575, tangentMode=RCTM_User, time=0.177, value=0.1465}
     {arriveTangent=-0.5331, interpMode=RCIM_Cubic, leaveTangent=-0.6001, tangentMode=RCTM_User, time=1, value=0.0085}
    ]}}
 }
}
