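# Sound cue for the looping air sound effect of an activated Fire Petal Shrine.
# A Random node picks among three looping Air Sprite shrine activation waves with equal weight, so the chosen variation repeats continuously.
# The cue takes its attenuation from the shrine's Activate settings asset, routes to the Fort loot reward sound class and plays at reduced volume.
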SoundCue /FirePetalShrineGameplay/Assets/Sounds/Cues/SC_FirePetalShrine_Activated_Air_SFX_Loop
root=Random_1
nodes [
 {id=Random_1, children [WavePlayer_0,WavePlayer_1,WavePlayer_2], properties {bShouldExcludeFromBranchCulling=true, weights [1,1,1]}}
 {
  id=WavePlayer_0
  properties {bLooping=true, soundWaveAssetPtr=/FirePetalShrineGameplay/Assets/Sounds/Waves/Air_Sprite_Shrine_Activation_SFX_01}
 }
 {
  id=WavePlayer_1
  properties {bLooping=true, soundWaveAssetPtr=/FirePetalShrineGameplay/Assets/Sounds/Waves/Air_Sprite_Shrine_Activation_SFX_02}
 }
 {
  id=WavePlayer_2
  properties {bLooping=true, soundWaveAssetPtr=/FirePetalShrineGameplay/Assets/Sounds/Waves/Air_Sprite_Shrine_Activation_SFX_03}
 }
]
properties {
 attenuationSettings=/FirePetalShrineGameplay/Assets/Sounds/Settings/AS_FirePetalShrine_Activate
 maxDistance=8600
 soundClassObject=&Fort_Loot_Reward_PSM
 volumeMultiplier=0.7
 attenuationOverrides {bAttenuate=false, bEnableReverbSend=false, bSpatialize=false}
 concurrencyOverrides {resolutionRule=PreventNew}
}
